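# Physics asset for the Fortnite DJ backpack cape cosmetic, providing collision bodies and joints for the cape's dynamic bones.
# A large capsule on backpack_root anchors a five-bone centre chain of narrow capsules.
# Two-bone left and right chains of wider, side-offset capsules branch from the last centre bone, dyn_cape05.
# All joints share the same limits, with moderate swing and very tight twist.
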PhysicsAsset /BRCosmetics/Accessories/FORT_Capes/M_MED_Fortnite_DJ/Meshes/F_MED_Fortnite_DJ_Backpack_Cape_Dyn
bodies {
 backpack_root [{shapeType=Capsule, length=41.375, radius=23.9062, center=<-0.0001 23.6688 -29.3732>, rotation=<rot 0 0 5.018>}]
 dyn_cape01 [{shapeType=Capsule, length=15.97, radius=4.5076, center=<4.0879 -2.3 0>}]
 dyn_cape01_l [{shapeType=Capsule, length=13.5012, radius=10, center=<4.0879 -6.8 -10>}]
 dyn_cape01_r [{shapeType=Capsule, length=13.5012, radius=10, center=<4.0879 -6.8 10>}]
 dyn_cape02 [{shapeType=Capsule, length=15.97, radius=4.5076, center=<4.0879 -2.3 0>}]
 dyn_cape02_l [{shapeType=Capsule, length=13.5012, radius=10, center=<4.0879 -5.8 -10>}]
 dyn_cape02_r [{shapeType=Capsule, length=13.5012, radius=10, center=<4.0879 -5.8 10>}]
 dyn_cape03 [{shapeType=Capsule, length=15.97, radius=4.5076, center=<4.0879 -2.3 0>}]
 dyn_cape04 [{shapeType=Capsule, length=15.97, radius=4.5076, center=<4.0879 -2.3 0>}]
 dyn_cape05 [{shapeType=Capsule, length=15.97, radius=4.5076, center=<4.0879 -2.3 0>}]
}
constraints [
 {bone1Name=dyn_cape02, bone2Name=dyn_cape01, swing1LimitDegrees=45, swing2LimitDegrees=45, twistLimitDegrees=5}
 {bone1Name=dyn_cape03, bone2Name=dyn_cape02, swing1LimitDegrees=45, swing2LimitDegrees=45, twistLimitDegrees=5}
 {bone1Name=dyn_cape04, bone2Name=dyn_cape03, swing1LimitDegrees=45, swing2LimitDegrees=45, twistLimitDegrees=5}
 {bone1Name=dyn_cape05, bone2Name=dyn_cape04, swing1LimitDegrees=45, swing2LimitDegrees=45, twistLimitDegrees=5}
 {bone1Name=dyn_cape02_l, bone2Name=dyn_cape01_l, swing1LimitDegrees=45, swing2LimitDegrees=45, twistLimitDegrees=5}
 {bone1Name=dyn_cape02_r, bone2Name=dyn_cape01_r, swing1LimitDegrees=45, swing2LimitDegrees=45, twistLimitDegrees=5}
 {bone1Name=dyn_cape01, bone2Name=backpack_root, swing1LimitDegrees=45, swing2LimitDegrees=45, twistLimitDegrees=5}
 {bone1Name=dyn_cape01_r, bone2Name=dyn_cape05, swing1LimitDegrees=45, swing2LimitDegrees=45, twistLimitDegrees=5}
 {bone1Name=dyn_cape01_l, bone2Name=dyn_cape05, swing1LimitDegrees=45, swing2LimitDegrees=45, twistLimitDegrees=5}
]
